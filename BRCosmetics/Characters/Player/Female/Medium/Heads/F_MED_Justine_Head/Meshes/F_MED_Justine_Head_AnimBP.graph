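# Animation Blueprint for the Justine medium female player head, targeting the shared Fortnite base 3-lateral head skeleton.
# It declares a head rotation, per-axis percentage variables with a min/max pair for Z, and an IsIvoryBriteEquipped flag.
# The event graph holds only an unconnected update event, and no logic shown reads or writes these variables.
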AnimBlueprint /BRCosmetics/Characters/Player/Female/Medium/Heads/F_MED_Justine_Head/Meshes/F_MED_Justine_Head_AnimBP
target_skeleton=/Game/Characters/Player/Common/Fortnite_Base_Head/Fortnite_Base_3L_Head_Skeleton
graphs [EventGraph {nodes [Event BlueprintUpdateAnimation()]}]
variables [
 headRot {type=Rotator}
 IsIvoryBriteEquipped {type=boolean}
 xHeadRotPer {type=number, default=0.3}
 yHeadRotPer {type=number, default=0.3}
 zHeadRotPerMax {type=number, default=0.8}
 zHeadRotPerMin {type=number}
]
